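# Base UI Kit toast notification widget showing a message, helper text and either a player avatar or status icon.
# PreConstruct applies ToastType, both texts and progress amount through setter functions.
# RefreshToast uses enum Selects: the first type shows avatar, Accept button and progress bar; others show a left status indicator.
# Backplate dynamic material parameters drive glow color and progress; intro/outro animations are defined but not played here.
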WidgetBlueprint /FortUILibrary/Systems/Toast/WBP_UIKit_Toast_Base
animations [
 Anim_Intro {
  binding_count=4
  length_seconds=0.6
  object=&self:Anim_Intro
  sequence {bindings [
    WBP_UIKit_Toast_Base_C_66 {
     id=#1
     possessed_class=WBP_UIKit_Toast_Base_C
     tracks [
      {
       type=MovieScene2DTransformTrack
       sections [{
         type=MovieScene2DTransformSection
         end_frame=11
         start_frame=0
         channels [Translation.X {type=MovieSceneScriptingFloatChannel, keys [[0,100],[4.0003,0]]}]
        }]
      }
      {
       type=MovieSceneFloatTrack
       display_name=Render Opacity
       sections [{type=MovieSceneFloatSection, end_frame=12, start_frame=0, channels [None {type=MovieSceneScriptingFloatChannel, keys [[0,0],[4.0003,1]]}]}]
      }
     ]
    }
    WBP_PlayerAvatar {
     id=#2
     possessed_class=WBP_PlayerAvatar_C
     tracks [{
       type=MovieScene2DTransformTrack
       sections [{
         type=MovieScene2DTransformSection
         end_frame=12
         start_frame=0
         channels [
          Scale.X {type=MovieSceneScriptingFloatChannel, keys [[7,0],[10,1.1],[12,1]]}
          Scale.Y {type=MovieSceneScriptingFloatChannel, keys [[7,0],[10,1.1],[12,1]]}
         ]
        }]
      }]
    }
    WBP_UIKit_StatusIndicator_Left {
     id=#3
     possessed_class=WBP_UIKit_StatusIndicator_C
     tracks [{
       type=MovieScene2DTransformTrack
       sections [{
         type=MovieScene2DTransformSection
         end_frame=12
         start_frame=0
         channels [
          Scale.X {type=MovieSceneScriptingFloatChannel, keys [[7,0],[10,1.1],[12,1]]}
          Scale.Y {type=MovieSceneScriptingFloatChannel, keys [[7,0],[10,1.1],[12,1]]}
         ]
        }]
      }]
    }
    Backplate {
     id=#4
     possessed_class=Image
     tracks [{
       type=MovieSceneWidgetMaterialTrack
       sections [{
         type=MovieSceneParameterSection
         end_frame=12
         start_frame=0
         channels [Animation Control {type=MovieSceneScriptingFloatChannel, keys [[0,0],[12,1]]}]
        }]
      }]
    }
   ]}
 }
 Anim_Outro {
  binding_count=1
  length_seconds=0.2
  object=&self:Anim_Outro
  sequence {bindings [WBP_UIKit_Toast_Base_C_95 {
     id=#5
     possessed_class=WBP_UIKit_Toast_Base_C
     tracks [
      {
       type=MovieScene2DTransformTrack
       sections [{
         type=MovieScene2DTransformSection
         end_frame=4
         start_frame=0
         channels [Translation.X {type=MovieSceneScriptingFloatChannel, keys [[0,0],[4,40]]}]
        }]
      }
      {
       type=MovieSceneFloatTrack
       display_name=Render Opacity
       sections [{type=MovieSceneFloatSection, end_frame=4, start_frame=0, channels [None {type=MovieSceneScriptingFloatChannel, keys [[0,1],[4,0]]}]}]
      }
     ]
    }]}
 }
]
graphs [
 EventGraph {nodes [
   Event PreConstruct()
   > InitializeToast()
   > SetMessageText!(execute, self, MessageText:Text=Get Message Text() -> (Message Text:Text)) -> (then)
   > SetHelperText!(execute, self, HelperText:Text=Get Helper Text() -> (Helper Text:Text)) -> (then)
   > SetProgressAmount!(execute, self, ProgressAmount:float=Get Progress Amount() -> (Progress Amount:float)) -> (then)
   Event Construct()
  ]}
 InitializeToast {nodes [
   InitializeToast() {is=title K2Node_FunctionEntry}
   > SetToastType!(execute, self, ToastType:E_UI_ToastType Enum=Get ToastType() -> (ToastType:E_UI_ToastType Enum=NewEnumerator0)) -> (then)
  ]}
 RefreshToast {nodes [
   RefreshToast() {is=title K2Node_FunctionEntry}
   > SetVisibility(
    self=Get WBP_PlayerAvatar() -> (WBP_PlayerAvatar:WBP Player Avatar Object Reference)
    InVisibility=Select!(
     NewEnumerator0:ESlateVisibility Enum=HitTestInvisible
     NewEnumerator1:ESlateVisibility Enum=Collapsed
     NewEnumerator2:ESlateVisibility Enum=Collapsed
     NewEnumerator3:ESlateVisibility Enum=Collapsed
     Index:E_UI_ToastType Enum=Get ToastType() -> (ToastType=NewEnumerator0)
    ) -> (ReturnValue)
   )
   > SetVisibility(
    self=Get WBP_UIKit_StatusIndicator_Left() -> (WBP_UIKit_StatusIndicator_Left:WBP UIKit Status Indicator Object Reference)
    InVisibility=Select(
     NewEnumerator0=Collapsed
     NewEnumerator1=SelfHitTestInvisible
     NewEnumerator2=SelfHitTestInvisible
     NewEnumerator3=SelfHitTestInvisible
     Index=Get ToastType() -> (ToastType=NewEnumerator0)
    )
   )
   > SetIcon!(
    execute
    self:WBP UIKit Status Indicator Base Object Reference=Get WBP_UIKit_StatusIndicator_Left()
    IconAsset:Object Reference=Select(
     NewEnumerator0:Object Reference
     NewEnumerator1:Object Reference=/FortUILibrary/Textures/MaterialInstances/MI_UIKit_Icon_Check
     NewEnumerator2:Object Reference=/FortUILibrary/Textures/MaterialInstances/MI_UIKit_Icon_ExclamationMark
     NewEnumerator3:Object Reference=&MI_UIKit_Icon_ExclamationMark
     Index=Get ToastType() -> (ToastType=NewEnumerator0)
    ) -> (ReturnValue:Object Reference)
   ) -> (then)
   > SetIndicatorType!(
    execute
    self:WBP UIKit Status Indicator Base Object Reference=Get WBP_UIKit_StatusIndicator_Left()
    IndicatorType:E_UI_StatusIndicator_Type Enum=Select(
     NewEnumerator0:E_UI_StatusIndicator_Type Enum=NewEnumerator0
     NewEnumerator1:E_UI_StatusIndicator_Type Enum=NewEnumerator0
     NewEnumerator2:E_UI_StatusIndicator_Type Enum=NewEnumerator4
     NewEnumerator3:E_UI_StatusIndicator_Type Enum=NewEnumerator2
     Index=Get ToastType() -> (ToastType=NewEnumerator0)
    ) -> (ReturnValue:E_UI_StatusIndicator_Type Enum)
   ) -> (then)
   > SetVisibility(
    self=Get WBP_UIKit_StatusIndicator_Right() -> (WBP_UIKit_StatusIndicator_Right:WBP UIKit Status Indicator Object Reference)
    InVisibility=Select(
     NewEnumerator0:ESlateVisibility Enum=Collapsed
     NewEnumerator1:ESlateVisibility Enum=Collapsed
     NewEnumerator2:ESlateVisibility Enum=Collapsed
     NewEnumerator3:ESlateVisibility Enum=Collapsed
     Index=Get ToastType() -> (ToastType=NewEnumerator0)
    ) -> (ReturnValue:ESlateVisibility Enum)
   )
   > SetIcon(
    self=Get WBP_UIKit_StatusIndicator_Right()
    IconAsset=Select(
     NewEnumerator0:Object Reference
     NewEnumerator1:Object Reference=&MI_UIKit_Icon_Check
     NewEnumerator2:Object Reference=&MI_UIKit_Icon_ExclamationMark
     NewEnumerator3:Object Reference
     Index=Get ToastType() -> (ToastType=NewEnumerator0)
    ) -> (ReturnValue:Object Reference)
   )
   > SetIndicatorType(
    self=Get WBP_UIKit_StatusIndicator_Right()
    IndicatorType=Select(
     NewEnumerator0:E_UI_StatusIndicator_Type Enum=NewEnumerator0
     NewEnumerator1:E_UI_StatusIndicator_Type Enum=NewEnumerator0
     NewEnumerator2:E_UI_StatusIndicator_Type Enum=NewEnumerator4
     NewEnumerator3:E_UI_StatusIndicator_Type Enum=NewEnumerator2
     Index=Get ToastType() -> (ToastType=NewEnumerator0)
    ) -> (ReturnValue:E_UI_StatusIndicator_Type Enum)
   )
   > SetVisibility(
    self=Get Divider() -> (Divider:Image Object Reference)
    InVisibility=Select(
     NewEnumerator0:ESlateVisibility Enum=HitTestInvisible
     NewEnumerator1:ESlateVisibility Enum=Collapsed
     NewEnumerator2:ESlateVisibility Enum=Collapsed
     NewEnumerator3:ESlateVisibility Enum=Collapsed
     Index=Get ToastType() -> (ToastType=NewEnumerator0)
    ) -> (ReturnValue:ESlateVisibility Enum)
   )
   > SetPadding!(
    execute
    self:Horizontal Box Slot Object Reference=SlotAsHorizontalBoxSlot!(Widget:Widget Object Reference=Get Divider()) -> (ReturnValue:Horizontal Box Slot Object Reference)
    InPadding_Left:float=0
    InPadding_Top:float=0
    InPadding_Right:float=0
    InPadding_Bottom:float=Select(
     NewEnumerator0:float=4
     NewEnumerator1:float=0
     NewEnumerator2:float=0
     NewEnumerator3:float=0
     Index=Get ToastType() -> (ToastType=NewEnumerator0)
    ) -> (ReturnValue:float)
   ) -> (then)
   > n1 = GetDynamicMaterial(self=Get Backplate() -> (Backplate:Image Object Reference))
   > SetScalarParameterValue(
    self=@n1
    ParameterName=Progress Bar Visibility
    Value=Select(NewEnumerator0=1, NewEnumerator1=0, NewEnumerator2=0, NewEnumerator3=0, Index=Get ToastType() -> (ToastType=NewEnumerator0))
   )
   > SetPadding(
    self=SlotAsHorizontalBoxSlot(Widget=Get LeftSideContent() -> (LeftSideContent:Horizontal Box Object Reference))
    InPadding_Left=16
    InPadding_Top=16
    InPadding_Right=16
    InPadding_Bottom=Select(NewEnumerator0=20, NewEnumerator1=16, NewEnumerator2=16, NewEnumerator3=16, Index=Get ToastType() -> (ToastType=NewEnumerator0))
   )
   > SetVisibility(
    self=Get RightSideContent() -> (RightSideContent:Horizontal Box Object Reference)
    InVisibility=Select(
     NewEnumerator0:ESlateVisibility Enum=HitTestInvisible
     NewEnumerator1:ESlateVisibility Enum=Collapsed
     NewEnumerator2:ESlateVisibility Enum=Collapsed
     NewEnumerator3:ESlateVisibility Enum=Collapsed
     Index=Get ToastType() -> (ToastType=NewEnumerator0)
    ) -> (ReturnValue:ESlateVisibility Enum)
   )
   > n2 = GetDynamicMaterial(self=Get Backplate())
   > SetVectorParameterValue(
    self=@n2
    ParameterName=Glow Color
    Value=Select(
     NewEnumerator0:Linear Color Structure=(R=1,G=1,B=1,A=1)
     NewEnumerator1:Linear Color Structure=(R=0.0395,G=1,B=0.2307,A=1)
     NewEnumerator2:Linear Color Structure=(R=1,G=0.147,B=0.1329,A=1)
     NewEnumerator3:Linear Color Structure=(R=0.9301,G=1,B=0.0103,A=1)
     Index=Get ToastType() -> (ToastType=NewEnumerator0)
    ) -> (ReturnValue:Linear Color Structure)
   )
   > SetVisibility(
    self=Get AcceptButton() -> (AcceptButton:WBP UIKit Button Generic Object Reference)
    InVisibility=Select(
     NewEnumerator0:ESlateVisibility Enum=Visible
     NewEnumerator1:ESlateVisibility Enum=Collapsed
     NewEnumerator2:ESlateVisibility Enum=Collapsed
     NewEnumerator3:ESlateVisibility Enum=Collapsed
     Index=Get ToastType() -> (ToastType=NewEnumerator0)
    ) -> (ReturnValue:ESlateVisibility Enum)
   )
  ]}
 SetHelperText {nodes [
   n1 = SetHelperText!() -> (then, HelperText:Text) {is=title K2Node_FunctionEntry}
   > Set Text(Text=@n1, self:Rich Text Block Object Reference=Get HelperText() -> (HelperText:Common Rich Text Block Object Reference))
  ]}
 SetMessageText {nodes [
   n1 = SetMessageText!() -> (then, MessageText:Text) {is=title K2Node_FunctionEntry}
   > Set Text(Text=@n1, self:Text Object Reference=Get Message() -> (Message:Common Text Object Reference))
  ]}
 SetProgressAmount {nodes [
   n1 = SetProgressAmount!() -> (then, ProgressAmount:float=~) {is=title K2Node_FunctionEntry}
   > n2 = GetDynamicMaterial(self=Get Backplate())
   > SetScalarParameterValue(self=@n2, ParameterName=Progress Amount, Value=@n1)
  ]}
 SetToastType {nodes [
   n1 = SetToastType!() -> (then, ToastType:E_UI_ToastType Enum=NewEnumerator0) {is=title K2Node_FunctionEntry}
   > Set ToastType(ToastType:E_UI_ToastType Enum=@n1) -> (Output_Get:E_UI_ToastType Enum=NewEnumerator0)
   > RefreshToast() {is=title K2Node_CallFunction}
  ]}
]
variables [
 Helper Text {type=string, default="Lorem ipsum dolor sit amet, consectetur adipiscing elit. Maecenas elementum iaculis massa sit amet pharetra."}
 Message Text {type=string, default="[PH] Message Goes Here"}
 Progress Amount {type=number, default=0.5}
 ToastType {type=E_UI_ToastType, default=External}
]
widgets {
 info {inheritedWidgetCount=0, parentClass=&UserWidget, rootWidgetClass=&Overlay, widgetCount=14}
 widgets [
  Overlay_32 {widgetClassPath=&Overlay, widget=Overlay_32}
  Backplate {
   widgetClassPath=&Image
   bIsVariable=true
   parent=Overlay_32
   widget=Backplate
   properties {
    visibility=HitTestInvisible
    brush {
     bIsDynamicallyLoaded=false
     drawAs=Image
     imageType=FullColor
     mirroring=NoMirror
     resourceName=None
     resourceObject=/FortUILibrary/Systems/Toast/MaterialInstances/MI_UIKit_Toast_Backplate
     tiling=NoTile
     imageSize=<32 32>
     margin=<ltrb 0 0 0 0>
     outlineSettings {
      bUseBrushTransparency=false
      roundingType=HalfHeightRadius
      width=0
      color {colorUseRule=UseColor_Specified, specifiedColor=<rgba 0 0 0 0>}
      cornerRadii=<0 0 0 0>
     }
     tintColor {colorUseRule=UseColor_Specified, specifiedColor=<rgba 1 1 1 1>}
     uVRegion {bIsValid=false, max=<0 0>, min=<0 0>}
    }
   }
   slot {class=OverlaySlot, properties {horizontalAlignment=HAlign_Fill, verticalAlignment=VAlign_Fill}}
  }
  VerticalBox_56 {widgetClassPath=&VerticalBox, parent=Overlay_32, widget=VerticalBox_56, slot {class=OverlaySlot}}
  HorizontalBox_196 {widgetClassPath=&HorizontalBox, parent=VerticalBox_56, widget=HorizontalBox_196, slot {class=VerticalBoxSlot}}
  LeftSideContent {
   widgetClassPath=&HorizontalBox
   bIsVariable=true
   parent=HorizontalBox_196
   widget=LeftSideContent
   slot {class=HorizontalBoxSlot, properties {padding=<ltrb 16 16 16 20>}}
  }
  WBP_PlayerAvatar {
   widgetClassPath=/Game/UI/Foundation/Profile/WBP_PlayerAvatar
   bIsVariable=true
   parent=LeftSideContent
   widget=WBP_PlayerAvatar
   properties {showOnlinePresence=false}
   slot {class=HorizontalBoxSlot, properties {horizontalAlignment=HAlign_Left, verticalAlignment=VAlign_Center, padding=<ltrb 0 0 16 0>}}
  }
  WBP_UIKit_StatusIndicator_Left {
   widgetClassPath=/FortUILibrary/UIKit/StatusIndicator/WBP_UIKit_StatusIndicator
   bIsVariable=true
   parent=LeftSideContent
   widget=WBP_UIKit_StatusIndicator_Left
   slot {class=HorizontalBoxSlot, properties {verticalAlignment=VAlign_Center, padding=<ltrb 0 0 16 0>}}
  }
  VerticalBox_53 {
   widgetClassPath=&VerticalBox
   parent=LeftSideContent
   widget=VerticalBox_53
   slot {class=HorizontalBoxSlot, properties {verticalAlignment=VAlign_Center}}
  }
  Message {
   widgetClassPath=&CommonTextBlock
   bIsVariable=true
   parent=VerticalBox_53
   widget=Message
   properties {
    text=Failed To Send Invite
    wrapTextAt=420
    font {
     bForceMonospaced=false
     bMaterialIsStencil=false
     fontMaterial=null
     fontObject=&HeadingNow
     letterSpacing=10
     monospacedWidth=1
     size=13.5
     skewAmount=0
     typefaceFontName=75 Medium
     outlineSettings {
      bApplyOutlineToDropShadows=false
      bMiteredCorners=false
      bSeparateFillAlpha=false
      outlineMaterial=null
      outlineSize=2
      outlineColor=<rgba 0 0 0 0.3>
     }
    }
    shadowOffset=<0 0>
   }
   slot {class=VerticalBoxSlot}
  }
  HelperText {
   widgetClassPath=&CommonRichTextBlock
   bIsVariable=true
   parent=VerticalBox_53
   widget=HelperText
   properties {
    applyLineHeightToBottomLine=false
    bOverrideDefaultStyle=true
    text="[PH] Lorem ipsum dolor sit amet, consectetur adipiscing elit. Pellentesque non facilisis odio, id lobortis dui."
    wrapTextAt=420
    defaultTextStyleOverride {
     overflowPolicy=Clip
     transformPolicy=None
     colorAndOpacity {colorUseRule=UseColor_Specified, specifiedColor=<rgba 1 1 1 0.7>}
     font {
      bForceMonospaced=false
      bMaterialIsStencil=false
      fontMaterial=null
      fontObject=&HeadingNow
      letterSpacing=0
      monospacedWidth=1
      size=13.5
      skewAmount=0
      typefaceFontName=74 Regular
      outlineSettings {
       bApplyOutlineToDropShadows=false
       bMiteredCorners=false
       bSeparateFillAlpha=false
       outlineMaterial=null
       outlineSize=2
       outlineColor=<rgba 0 0 0 0.3>
      }
     }
     highlightColor {colorUseRule=UseColor_Specified, specifiedColor=<rgba 0 0 0 1>}
     highlightShape {
      bIsDynamicallyLoaded=false
      drawAs=Image
      imageType=NoImage
      mirroring=NoMirror
      resourceName=None
      resourceObject=null
      tiling=NoTile
      imageSize=<32 32>
      margin=<ltrb 0 0 0 0>
      outlineSettings {
       bUseBrushTransparency=false
       roundingType=HalfHeightRadius
       width=0
       color {colorUseRule=UseColor_Specified, specifiedColor=<rgba 0 0 0 0>}
       cornerRadii=<0 0 0 0>
      }
      tintColor {colorUseRule=UseColor_Specified, specifiedColor=<rgba 1 1 1 1>}
      uVRegion {bIsValid=false, max=<0 0>, min=<0 0>}
     }
     selectedBackgroundColor {colorUseRule=UseColor_Foreground, specifiedColor=<rgba 1 0 1 1>}
     shadowColorAndOpacity=<rgba 0 0 0 1>
     shadowOffset=<0 0>
     strikeBrush {
      bIsDynamicallyLoaded=false
      drawAs=Image
      imageType=NoImage
      mirroring=NoMirror
      resourceName=None
      resourceObject=null
      tiling=NoTile
      imageSize=<32 32>
      margin=<ltrb 0 0 0 0>
      outlineSettings {
       bUseBrushTransparency=false
       roundingType=HalfHeightRadius
       width=0
       color {colorUseRule=UseColor_Specified, specifiedColor=<rgba 0 0 0 0>}
       cornerRadii=<0 0 0 0>
      }
      tintColor {colorUseRule=UseColor_Specified, specifiedColor=<rgba 1 1 1 1>}
      uVRegion {bIsValid=false, max=<0 0>, min=<0 0>}
     }
     underlineBrush {
      bIsDynamicallyLoaded=false
      drawAs=Image
      imageType=NoImage
      mirroring=NoMirror
      resourceName=None
      resourceObject=null
      tiling=NoTile
      imageSize=<32 32>
      margin=<ltrb 0 0 0 0>
      outlineSettings {
       bUseBrushTransparency=false
       roundingType=HalfHeightRadius
       width=0
       color {colorUseRule=UseColor_Specified, specifiedColor=<rgba 0 0 0 0>}
       cornerRadii=<0 0 0 0>
      }
      tintColor {colorUseRule=UseColor_Specified, specifiedColor=<rgba 1 1 1 1>}
      uVRegion {bIsValid=false, max=<0 0>, min=<0 0>}
     }
    }
   }
   slot {class=VerticalBoxSlot}
  }
  Divider {
   widgetClassPath=&Image
   bIsVariable=true
   parent=HorizontalBox_196
   widget=Divider
   properties {
    visibility=HitTestInvisible
    brush {
     bIsDynamicallyLoaded=false
     drawAs=Image
     imageType=NoImage
     mirroring=NoMirror
     resourceName=None
     resourceObject=null
     tiling=NoTile
     imageSize=<2 32>
     margin=<ltrb 0 0 0 0>
     outlineSettings {
      bUseBrushTransparency=false
      roundingType=HalfHeightRadius
      width=0
      color {colorUseRule=UseColor_Specified, specifiedColor=<rgba 0 0 0 0>}
      cornerRadii=<0 0 0 0>
     }
     tintColor {colorUseRule=UseColor_Specified, specifiedColor=<rgba 1 1 1 0.05>}
     uVRegion {bIsValid=false, max=<0 0>, min=<0 0>}
    }
   }
   slot {class=HorizontalBoxSlot, properties {padding=<ltrb 0 0 0 4>}}
  }
  RightSideContent {
   widgetClassPath=&HorizontalBox
   bIsVariable=true
   parent=HorizontalBox_196
   widget=RightSideContent
   slot {class=HorizontalBoxSlot, properties {horizontalAlignment=HAlign_Center, verticalAlignment=VAlign_Center, padding=<ltrb 16 0 16 0>}}
  }
  WBP_UIKit_StatusIndicator_Right {
   widgetClassPath=&WBP_UIKit_StatusIndicator
   bIsVariable=true
   parent=RightSideContent
   widget=WBP_UIKit_StatusIndicator_Right
   slot {class=HorizontalBoxSlot}
  }
  AcceptButton {
   widgetClassPath=&WBP_UIKit_Button_Generic
   bIsVariable=true
   parent=RightSideContent
   widget=AcceptButton
   properties {
    bDisplayInputActionWhenNotInteractable=true
    text=Accept
    textTransformPolicy=None
    useBackgroundBlock=false
    buttonPadding=<ltrb 8 0 8 0>
    font {
     bForceMonospaced=false
     bMaterialIsStencil=false
     fontMaterial=/FortUILibrary/UIKit/Buttons/MaterialInstances/CTAButtons/MI_UIKit_TextQuiet_Prompt
     fontObject=&HeadingNow
     letterSpacing=48
     monospacedWidth=1
     size=15.75
     skewAmount=0
     typefaceFontName=74 Regular
     outlineSettings {
      bApplyOutlineToDropShadows=false
      bMiteredCorners=false
      bSeparateFillAlpha=false
      outlineMaterial=null
      outlineSize=2
      outlineColor=<rgba 0 0 0 0.3>
     }
    }
   }
   slot {class=HorizontalBoxSlot}
  }
 ]
}
